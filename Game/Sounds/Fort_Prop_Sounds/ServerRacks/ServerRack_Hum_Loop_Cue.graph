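# Sound cue for a looping ambient hum on server rack props.
# A looping wave player plays the server rack tower hum through a Modulator.
# It uses a dedicated hum attenuation asset and the Fort_Ambient_Props_Athena sound class at low volume and reduced priority.
# A concurrency override caps simultaneous instances to a few, limiting stacking when many racks are nearby.
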SoundCue /Game/Sounds/Fort_Prop_Sounds/ServerRacks/ServerRack_Hum_Loop_Cue
root=Modulator_0
nodes [
 {id=Modulator_0, children [WavePlayer_0]}
 {
  id=WavePlayer_0
  properties {bLooping=true, soundWaveAssetPtr=&/Computer_Server_Rack_Tower_Hum_Loop_01}
 }
]
properties {
 attenuationSettings=&/ServerRack_Hum_Att
 bOverrideConcurrency=true
 maxDistance=1900
 priority=0.25
 soundClassObject=&Fort_Ambient_Props_Athena
 volumeMultiplier=0.15
 concurrencyOverrides {maxCount=3}
}
