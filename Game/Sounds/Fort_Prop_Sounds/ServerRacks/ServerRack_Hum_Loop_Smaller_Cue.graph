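# Sound cue for a quiet looping server rack hum, an indoor ambient prop sound in the "Smaller" variant.
# A looping wave feeds a modulator, then an enveloper that ramps volume up from silence as it starts.
# Inline attenuation overrides the referenced attenuation asset, giving a short audible range and a distance-scaled send to the distant reverb submix.
# A concurrency override caps simultaneous instances, and low priority and volume keep it in the background.
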SoundCue /Game/Sounds/Fort_Prop_Sounds/ServerRacks/ServerRack_Hum_Loop_Smaller_Cue
root=Enveloper_0
nodes [
 {
  id=Enveloper_0
  children [Modulator_0]
  properties {pitchCurve {editorCurveData {keys [{time=0}]}}, volumeCurve {editorCurveData {keys [<at 0 0>,<at 1 1>]}}}
 }
 {id=Modulator_0, children [WavePlayer_0]}
 {
  id=WavePlayer_0
  properties {bLooping=true, soundWaveAssetPtr=&/Computer_Server_Rack_Tower_Hum_Loop_01}
 }
]
properties {
 attenuationSettings=&/ServerRack_Hum_Att
 bOverrideAttenuation=true
 bOverrideConcurrency=true
 maxDistance=900
 priority=0.25
 soundClassObject=&Fort_Ambient_Indoors_Athena
 volumeMultiplier=0.15
 attenuationOverrides {
  bEnableSubmixSends=true
  falloffDistance=650
  attenuationShapeExtents {x=250}
  submixSendSettings [{
    bEnableHPFCutoff=false
    bEnableLPFCutoff=false
    disableManualSendClamp=false
    hPFCutoff=20
    lPFCutoff=20000
    maxSendDistance=6000
    maxSendLevel=1
    minSendDistance=400
    minSendLevel=0
    sendLevel=0.6
    sendLevelControlMethod=Linear
    soundSubmix=&Fortnite_Reverb_Distant
   }]
 }
 concurrencyOverrides {maxCount=3}
}
